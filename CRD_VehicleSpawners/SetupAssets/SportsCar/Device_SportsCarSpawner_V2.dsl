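; Creative sports car spawner device that adds color/style and vehicle speed tuning options to a parent spawner.
; On initialization it copies the chosen physics config and curve table pair into the spawner and loads the table.
; Version upgrades async-load Chapter 4 or 5 configs for older saves on default tunings, and move certain outdated selections to current class defaults.
; Helper functions convert the color enum to and from an index.
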(event EventBeginPlay
  (|CallParentFunction))

(event Collision|EventActorBeginOverlap
  (params (OtherActor :type Actor%20Object%20Reference)))

(event EventTick
  (params (DeltaSeconds :type Float%20%28single-precision%29)))

(event HandleVersionUpgrades(CreativeVersion)
  (params (LastSavedVersion :type Creative%20Version%20Id%20Structure))
  (bind (_vehiclecurvetablereference_6_b029e36547ba31c4746d8383a8dd52af _physicsvehicleconfigclass_5_28e9292b488ba4c2ffe9458c47023c0c) (Utilities|Struct|BreakVehicleConfigCurveCombo (Variables|UserOptions|GetVehicle%20Speed%20Tunings)))
  (bind (_vehiclecurvetablereference_6_b029e36547ba31c4746d8383a8dd52af_2 _physicsvehicleconfigclass_5_28e9292b488ba4c2ffe9458c47023c0c_2) (Utilities|Struct|BreakVehicleConfigCurveCombo (Variables|UserOptions|GetVehicle%20Speed%20Tunings)))
  (bind _vehicle_speed_tunings (Class|GetClassDefaults "/CRD_VehicleSpawners/SetupAssets/SportsCar/Device_SportsCarSpawner_V2.Device_SportsCarSpawner_V2_C"))
  (bind (_vehicle_speed_tunings_vehiclecurvetablereference_6_b029e36547ba31c4746d8383a8dd52af _vehicle_speed_tunings_physicsvehicleconfigclass_5_28e9292b488ba4c2ffe9458c47023c0c) (Utilities|Struct|BreakVehicleConfigCurveCombo _vehicle_speed_tunings))
  (if (CallFunction|VehicleSpeedTuningsAtDefault)
    (CreativeVersion|CompareVersiontoTag LastSavedVersion "(TagName=\"Creative.Version.VehicleSpawners.SpeedTunings.Default.Ch5\")"
      (:OlderVersion
        (bind _class (Utilities|AsyncLoadClassAsset (Variables|Default|GetPhysicsConfigDefault_CH4))
          (:Completed
            (Variables|UserOptions|SetVehicle%20Speed%20Tunings (Utilities|Struct|MakeVehicleConfigCurveCombo "/Valet/DataTables/ValetGameData.ValetGameData" (Utilities|Casting|CastToFortPhysicsVehicleConfigsClass _class)))
            (OptionData|RefreshPropertyOverrides (Variables|Default|GetToyOptionsComponent)))))
      (:Equal
        _)
      (:NewerVersion
        _))
    (CreativeVersion|CompareVersiontoTag LastSavedVersion "(TagName=\"Creative.Version.VehicleSpawners.SpeedTunings.Default.Ch5S3\")"
      (:OlderVersion
        (bind _class_2 (Utilities|AsyncLoadClassAsset (Variables|Default|GetPhysicsConfigDefault_CH5))
          (:Completed
            (Variables|UserOptions|SetVehicle%20Speed%20Tunings (Utilities|Struct|MakeVehicleConfigCurveCombo "/Valet/DataTables/ValetGameData_CH5S1.ValetGameData_CH5S1" (Utilities|Casting|CastToFortPhysicsVehicleConfigsClass _class_2)))
            (OptionData|RefreshPropertyOverrides (Variables|Default|GetToyOptionsComponent))))))
    (else
      (CreativeVersion|CompareVersiontoTag LastSavedVersion "(TagName=\"Creative.Version.Release.33.00\")"
        (:OlderVersion
          (if (== _physicsvehicleconfigclass_5_28e9292b488ba4c2ffe9458c47023c0c_2 (Variables|Default|GetPhysicsConfigDefault_BR))
            _))
        (:NewerVersion
          (CreativeVersion|CompareVersiontoTag LastSavedVersion "(TagName=\"Creative.Version.Release.40.40\")"
            (:OlderVersion
              _)
            (:Equal
              _))
          (if (== _physicsvehicleconfigclass_5_28e9292b488ba4c2ffe9458c47023c0c (Variables|Default|GetPhysicsConfigDefault_CH5S2))
            _)))
      (Variables|UserOptions|SetVehicle%20Speed%20Tunings (Utilities|Struct|MakeVehicleConfigCurveCombo _vehicle_speed_tunings_vehiclecurvetablereference_6_b029e36547ba31c4746d8383a8dd52af _vehicle_speed_tunings_physicsvehicleconfigclass_5_28e9292b488ba4c2ffe9458c47023c0c))
      (OptionData|RefreshPropertyOverrides (Variables|Default|GetToyOptionsComponent)))))

(fn ColorAndStyleAsAnInteger
  (return (Math|Integer|SelectInt 14 (- (Math|Conversions|ToInteger(Byte) (Variables|UserOptions|Basic|GetColor%20and%20Style)) 1) (Utilities|Enum|Equal(Enum) (Variables|UserOptions|Basic|GetColor%20and%20Style) "NewEnumerator27"))))

(fn InitializeDevice
  (CallFunction|UpdateVehicleTuning)
  (CallFunction|LoadCurveTableIfAble)
  (|CallParentFunction))

(fn SetColorOption
  (params (ColorIndex :type Integer))
  (Variables|UserOptions|Basic|SetColor%20and%20Style (Utilities|Enum|BytetoEnumEColorSportsCar (Math|Conversions|ToByte(Integer) (+ ColorIndex 1)))))

(fn UpdateVehicleTuning
  (bind _vehicle_speed_tunings (Variables|UserOptions|GetVehicle%20Speed%20Tunings))
  (bind (_vehiclecurvetablereference_6_b029e36547ba31c4746d8383a8dd52af _physicsvehicleconfigclass_5_28e9292b488ba4c2ffe9458c47023c0c) (Utilities|Struct|BreakVehicleConfigCurveCombo _vehicle_speed_tunings))
  (Variables|Default|SetPhysics%20Vehicle%20Config%20Class _physicsvehicleconfigclass_5_28e9292b488ba4c2ffe9458c47023c0c)
  (Variables|Default|SetVehicle%20Curve%20Table%20Reference _vehiclecurvetablereference_6_b029e36547ba31c4746d8383a8dd52af)
  (CallFunction|LoadCurveTableIfAble))

(fn ConstructionScript
  (|CallParentFunction))

(fn VehicleSpeedTuningsatDefault
  (bind (_vehiclecurvetablereference_6_b029e36547ba31c4746d8383a8dd52af _physicsvehicleconfigclass_5_28e9292b488ba4c2ffe9458c47023c0c) (Utilities|Struct|BreakVehicleConfigCurveCombo (Variables|UserOptions|GetVehicle%20Speed%20Tunings)))
  (bind _vehicle_speed_tunings (Class|GetClassDefaults "/CRD_VehicleSpawners/SetupAssets/SportsCar/Device_SportsCarSpawner_V2.Device_SportsCarSpawner_V2_C"))
  (bind (_vehicle_speed_tunings_vehiclecurvetablereference_6_b029e36547ba31c4746d8383a8dd52af _vehicle_speed_tunings_physicsvehicleconfigclass_5_28e9292b488ba4c2ffe9458c47023c0c) (Utilities|Struct|BreakVehicleConfigCurveCombo _vehicle_speed_tunings))
  (return (and (== _vehicle_speed_tunings_vehiclecurvetablereference_6_b029e36547ba31c4746d8383a8dd52af _vehiclecurvetablereference_6_b029e36547ba31c4746d8383a8dd52af) (== _vehicle_speed_tunings_physicsvehicleconfigclass_5_28e9292b488ba4c2ffe9458c47023c0c _physicsvehicleconfigclass_5_28e9292b488ba4c2ffe9458c47023c0c))))
